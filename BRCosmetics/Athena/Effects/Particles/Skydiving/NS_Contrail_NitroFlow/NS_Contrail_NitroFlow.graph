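# NS_Contrail_NitroFlow is a Niagara skydiving contrail cosmetic that trails glowing goop ribbons, drips and meshes from the player's skeletal mesh.
# Ribbon, sprite-drip and mesh emitters follow sampled bones and sockets, including the hands, with spawn groups splitting particles per attachment.
# Diving attributes and timelines drive material panning, velocity and fades.
# Waveforms and curves pulse widths and scales, while drip and goop-ribbon emitters also adjust to time of day.
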NiagaraSystem /BRCosmetics/Athena/Effects/Particles/Skydiving/NS_Contrail_NitroFlow/NS_Contrail_NitroFlow
dependencies {
 usedDataInterfaces [&NiagaraDataInterfaceCurve, &NiagaraDataInterfaceArrayFloat, &NiagaraDataInterfaceColorCurve]
 usedDynamicInputs [
  &UniformRangedFloat
  &FN_FrontEndFloatModifier
  &SetBoolByFloatComparison
 ]
 usedModules [
  /Game/Effects/Niagara/Modules/Skydiving/FN_DivingAttributes/v4/FN_DivingAttributes_v4
  &SystemState
  &self:SystemSpawnScript.SystemScriptSource.SystemScriptGraph.NiagaraNodeAssignment_0.SetVariables_#1
  &self:SystemSpawnScript.SystemScriptSource.SystemScriptGraph.NiagaraNodeAssignment_1.SetVariables_#2
  &self:SystemSpawnScript.SystemScriptSource.SystemScriptGraph.NiagaraNodeAssignment_2.SetVariables_#3
  &self:SystemSpawnScript.SystemScriptSource.SystemScriptGraph.NiagaraNodeAssignment_3.SetVariables_#4
  &FN_Timer
  &Timeline
 ]
}
emitters {
 CentralRibbons {
  emitterUpdateScript [
   &EmitterState()
   /Game/Effects/Niagara/Modules/Spawn/FN_SpawnByEngineDistance(Multiplier=2, Step=5)
   /SetVariables(Emitter.R2=0.33, Emitter.R3=0.5, Emitter.R4=0.66) {isSetParameters=true}
   /Game/Effects/Niagara/Modules/Skydiving/FN_DivingAttributes/FN_EmitterDivingAttributes_v1(Accumulate Dive=-0.0025, Accumulate Rest=0, SkeletalMesh:NiagaraDataInterfaceSkeletalMesh=$User.SkeletalMesh)
   /SetVariables(Emitter.DistortionPan:NiagaraFloat=&Lerp_Float(A=-0.03, B=-0.035, Alpha=&MakeCustomFloatFromBool(Bool=$System.IsDiving, True Float=1, False Float=0))) {isSetParameters=true}
   /SetVariables(Emitter.AlphaPan:NiagaraFloat=&Lerp_Float(A=-0.3, B=-0.35, Alpha=&MakeCustomFloatFromBool(Bool=$System.IsDiving, True Float=1, False Float=0))) {isSetParameters=true}
  ]
  particleSpawnScript [
   &InitializeRibbon(
    Lifetime=0.75
    Module.Write Color=true
    Module.Write Lifetime=true
    Module.Write Mass=true
    Module.Write Material Random=true
    Module.Write Position=false
    Module.Write RibbonTwist=false
    Module.Write RibbonWidth=true
    Ribbon Width=25
   )
   /SetVariables(Particles.Stream:NiagaraInt32=&SpawnGroup()) {isSetParameters=true}
   &FN_SetSpawnGroup(Number of Spawn Groups=2)
   &AddVelocity(Velocity=<0 0 50>)
   /SetVariables(Particles.RibbonNumber:NiagaraInt32=$Particles.SpawnGroup) {isSetParameters=true}
  ]
  particleUpdateScript [
   &ParticleState()
   &SkeletalMeshLocation(
    Bone / Socket Index=$Particles.SpawnGroup
    MaskSpawnGroup=false
    Mesh Sampling Type=::NewEnumerator5
    Orientation Sampling=::NewEnumerator1
    Position Sampling=::NewEnumerator2
    Sampled Position Offset=&MaskVectorBySpawnGroup(Masked Vector=<-10 10 7.5>, Passthrough Vector=<10 10 7.5>)
    Skeletal Mesh=$User.Skelmesh
    Socket Sampling Mode=::NewEnumerator5
   )
   &InheritSourceMovement(
    Applied Force Scale=&VectorFromFloat(Value=-0.5)
    ApplyForce=true
    ApplyForceFalloff=false
    ApplyPosition=true
    ApplyPositionFalloff=false
    Module.LimitForceSpeed=false
    Source Velocity:Vector3f=$Particles.SkeletalMeshLocation.SampledVelocity
   )
   &SolveForcesAndVelocity(Clamp Velocity=false, Limit Acceleration=false)
   &self:CentralRibbons_0.NiagaraScratchPadContainer_1.FN_Lock_Ribbon_Head_0(
    Emitter Count:NiagaraInt32=$Emitter.Count
    Locked Position:Vector3f=$Particles.SkeletalMeshLocation.SampledPosition
    Particle Stream:NiagaraInt32=$Particles.Stream
   ) {module=FN_Lock_Ribbon_Head}
   &DynamicMaterialParameters(
    Index 0 Param 1=&Add_Float(A=&MaskFloatBySpawnGroup(Masked Float=$Emitter.R3, Passthrough Float=$Emitter.R4, Spawn Group Mask=2), B=0)
    Index 0 Param 2=&FloatFromCurve(
     FloatCurve=DataInterface() {propertyValues {Curve {keys [
         {arriveTangent=0.0098, interpMode=RCIM_Cubic, leaveTangent=0.0098, tangentMode=RCTM_User, time=0.2, value=0.2}
         {arriveTangent=-0.0177, interpMode=RCIM_Cubic, leaveTangent=0.0167, tangentMode=RCTM_Break, time=0.5, value=0.0125}
        ]}}}
     Scale Curve=&Waveform(
      Global Phase=&Add_Float(
       A=$Particles.NormalizedAge
       B=&RandomRangeFloat(Minimum=0, Maximum=1, Randomness Mode=::NewEnumerator1, Random Seed=$Particles.SpawnGroup, Fixed Random Seed=true)
      )
      "Waveform [1]"=::NewEnumerator0
      "[1] Frequency"=1
      "[1] Amplitude Min/Max"=<0 1>
      "[1] Bias"=0.75
     )
    )
    Index 0 Param 3=&Lerp_Float(A=-1, B=-1, Alpha=&MakeCustomFloatFromBool(Bool=$System.IsDiving, True Float=1, False Float=0))
    Module.Param0WriteEnabled=true
    Module.Param1WriteEnabled=true
    Module.Param2WriteEnabled=true
    Module.Param3WriteEnabled=true
   )
   &ScaleRibbonWidth(Ribbon Width Scale=&Multiply_Float(
     A=&Multiply_Float(
      A=&Add_Float(
       A=&FloatFromCurve(FloatCurve=DataInterface() {propertyValues {Curve {keys [
            <at 0 0.3 RCIM_Cubic>
            {arriveTangent=1.7947, interpMode=RCIM_Cubic, leaveTangent=1.7947, tangentMode=RCTM_User, time=0.2, value=0.9}
            {arriveTangent=-1, interpMode=RCIM_Cubic, leaveTangent=-1, time=0.4, value=1.05}
            {arriveTangent=-0.6667, interpMode=RCIM_Cubic, leaveTangent=-0.6667, time=0.5, value=0.6}
            {arriveTangent=-1.5, interpMode=RCIM_Cubic, leaveTangent=-1.5, time=0.7, value=0.85}
            <at 0.9 0 RCIM_Cubic>
           ]}}})
       B=0
      )
      B=&FloatFromCurve(FloatCurve=DataInterface() {propertyValues {Curve {keys [{interpMode=RCIM_Cubic, tangentMode=RCTM_User, time=0.75, value=1}, {interpMode=RCIM_Cubic, tangentMode=RCTM_User, time=1, value=0.5}]}}})
     )
     B=1
    )) {module=ScaleRibbonWidth001}
   &ScaleColor(
    Scale Alpha=&Multiply_Float(
     A=&Lerp_Float(A=1, B=0, Alpha=&MakeCustomFloatFromBool(Bool=$System.IsDiving, True Float=1, False Float=0))
     B=&Lerp_Float(A=1, B=0, Alpha=$System.Timeline001.NormalizedTime)
    )
    ScaleA=true
    ScaleRGB=true
   )
  ]
  rendererClasses [&NiagaraRibbonRendererProperties]
  renderers [{rendererClass=&NiagaraRibbonRendererProperties}]
 }
 DripsCenter {
  emitterUpdateScript [
   &EmitterState()
   &SpawnRate(SpawnRate=7, Use Spawn Probability=false)
  ]
  particleSpawnScript [
   &FN_SetSpawnGroup()
   &InitializeParticle(
    Color Mode=::NewEnumerator0
    Lifetime Min=1.8
    Lifetime Mode=::NewEnumerator1
    Mesh Scale Mode=::NewEnumerator3
    Mesh Uniform Scale=&Lerp_Float(
     A=0.12
     B=0.2
     Alpha=&FloatFromCurve(FloatCurve=DataInterface() {propertyValues {Curve {keys [<at 0 0 RCIM_Cubic>, {arriveTangent=1.5, interpMode=RCIM_Cubic, leaveTangent=1.5, time=0.9, value=0.775}, <at 1 1.5 RCIM_Cubic>]}}})
    )
    Sprite Size Mode=::NewEnumerator3
    Uniform Sprite Size=&Lerp_Float(
     A=0.12
     B=0.2
     Alpha=&FloatFromCurve(FloatCurve=DataInterface() {propertyValues {Curve {keys [<at 0 0 RCIM_Cubic>, {arriveTangent=1.5, interpMode=RCIM_Cubic, leaveTangent=1.5, time=0.9, value=0.775}, <at 1 1.5 RCIM_Cubic>]}}})
    )
    UsePositionOffset=false
    Write Material Random=true
   )
   &SkeletalMeshLocation(
    Bone / Socket Index=3
    Bone Sampling Mode=::NewEnumerator4
    MaskSpawnGroup=false
    Orientation Sampling=::NewEnumerator1
    Sampled Position Offset Coordinate Space:ENiagaraSkelMeshTransforms=::NewEnumerator0
    Skeletal Mesh=$User.SkeletalMesh
   )
  ]
  particleUpdateScript [
   &ParticleState()
   &ScaleSpriteSize(
    Uniform Curve Scale=20
    Uniform Curve Sprite Scale=DataInterface() {propertyValues {Curve {keys [
        {interpMode=RCIM_Cubic, tangentMode=RCTM_User, time=0, value=0}
        {interpMode=RCIM_Cubic, tangentMode=RCTM_User, time=0.5, value=1}
        {tangentMode=RCTM_User, time=1, value=0}
       ]}}}
   )
   &ScaleMeshSize(Scale Factor=&VectorFromFloat(Value=&FloatFromCurve(FloatCurve=DataInterface() {propertyValues {Curve {keys [
          {arriveTangent=6.4608, interpMode=RCIM_Cubic, leaveTangent=6.4608, tangentMode=RCTM_User, time=0, value=0}
          {interpMode=RCIM_Cubic, tangentMode=RCTM_User, time=0.1403, value=1}
          {arriveTangent=-0.0754, interpMode=RCIM_Cubic, leaveTangent=-0.0754, tangentMode=RCTM_User, time=0.7, value=1}
          {tangentMode=RCTM_User, time=1, value=0}
         ]}}})))
   &SkeletalMeshLocation(
    Bone / Socket Index=3
    Bone Sampling Mode=::NewEnumerator4
    MaskSpawnGroup=false
    Orientation Sampling=::NewEnumerator1
    Position Sampling=::NewEnumerator1
    Sampled Position Offset=<0 5 0>
    Sampled Position Offset Coordinate Space:ENiagaraSkelMeshTransforms=::NewEnumerator2
    Skeletal Mesh=$User.SkeletalMesh
   ) {module=SkeletalMeshLocation001}
   &ShapeLocation(
    Additional Quaternion Rotation=true
    Additional Yaw / Pitch / Roll=<0 0.25 0>
    Handle Radius=7
    Large Radius=20
    MaskSpawnGroup=false
    Override Seed=false
    Rotation Mode=::NewEnumerator3
    Rotation Quaternion=$Particles.SkeletalMeshLocation001.SampledRotation
    Shape Origin=$Particles.SkeletalMeshLocation001.SampledPosition
    Shape Primitive=::NewEnumerator3
    UseOwnerScale=true
   )
   &VortexForce(
    Origin Pull Amount=200
    UseOriginOffset=false
    Vortex Axis=&RotateVectorByQuaternion(Vector To Rotate=<1 0 0>, Quaternion=$Particles.SkeletalMeshLocation001.SampledRotation)
    Vortex Force Amount=500
    Vortex Origin:NiagaraPosition=$Particles.SkeletalMeshLocation001.SampledPosition
   )
   &SolveForcesAndVelocity(Clamp Velocity=false, Limit Acceleration=false)
   &ScaleSpriteSizeBySpeed(Max Scale Factor=<1.1 1.2>)
   &ScaleMeshSizeBySpeed(Max Scale Factor=<1 1 1>)
   &CameraOffset(Camera Offset Amount=&RandomRangeFloat(Minimum=10, Maximum=20))
   &Color(Color=&MultiplyLinearColor(
     Linear Color=&FN_Select_Color_from_Time_Of_Day(Morning=<rgba 1 1 1 1>, Day=<rgba 1 1 1 1>, Evening=<rgba 1 1 1 1>, Night=<rgba 0.5 0.5 0.5 1>)
     Scale Factor=&MakeVector4_FromLinearColor(LinearColor=&LerpLinearColors(StartColor=<rgba 0 3.5 2 1>, EndColor=<rgba 1 1 1 1>, LerpFactor=$Particles.NormalizedAge))
    )) {module=Color001}
  ]
  rendererClasses [&NiagaraSpriteRendererProperties]
  renderers [{rendererClass=&NiagaraSpriteRendererProperties}]
 }
 DripsHands {
  emitterUpdateScript [
   &EmitterState()
   &SpawnRate(SpawnRate=30, Use Spawn Probability=false)
  ]
  particleSpawnScript [
   &FN_SetSpawnGroup(Number of Spawn Groups=2)
   &InitializeParticle(
    Color Mode=::NewEnumerator0
    Lifetime Max=0.85
    Lifetime Min=0.6
    Lifetime Mode=::NewEnumerator1
    Mesh Scale Mode=::NewEnumerator1
    Mesh Uniform Scale Max=0.12
    Mesh Uniform Scale Min=0.07
    Sprite Size Mode=::NewEnumerator3
    Uniform Sprite Size=&Multiply_Float(
     A=&Lerp_Float(
      A=0.12
      B=0.2
      Alpha=&FloatFromCurve(
       FloatCurve=DataInterface() {propertyValues {Curve {keys [<at 0 0 RCIM_Cubic>, {arriveTangent=1.5, interpMode=RCIM_Cubic, leaveTangent=1.5, time=0.9, value=0.775}, <at 1 1.5 RCIM_Cubic>]}}}
       CurveIndex=&RandomRangeFloat(Minimum=0, Maximum=1)
      )
     )
     B=15
    )
    UsePositionOffset=false
    Write Material Random=true
   )
   &AddVelocity(Velocity=<0 0 150>, Velocity Speed Scale=&RandomRangeFloat(Minimum=0.8, Maximum=1))
   &SkeletalMeshLocation(
    Bone / Socket Index=$Particles.SpawnGroup
    MaskSpawnGroup=false
    Mesh Sampling Type=::NewEnumerator5
    Orientation Sampling=::NewEnumerator1
    Sampled Position Offset=&RandomRangeVector(Minimum=<-10 -10 -10>, Maximum=<10 10 10>)
    Sampled Position Offset Coordinate Space:ENiagaraSkelMeshTransforms=::NewEnumerator0
    Skeletal Mesh=$User.SkeletalMesh
    Socket Sampling Mode=::NewEnumerator5
   )
  ]
  particleUpdateScript [
   &ParticleState()
   &SkeletalMeshLocation(
    Bone / Socket Index=$Particles.SpawnGroup
    MaskSpawnGroup=false
    Mesh Sampling Type=::NewEnumerator5
    Orientation Sampling=::NewEnumerator1
    Position Sampling=::NewEnumerator1
    Sampled Position Offset=&RandomRangeVector(
     Minimum=<-5 -5 0>
     Maximum=&MaskVectorBySpawnGroup(Masked Vector=<5 5 20>, Passthrough Vector=<5 5 10>, Spawn Group Mask=&ModuloInt(A=$Particles.SpawnGroup, B=2))
    )
    Sampled Position Offset Coordinate Space:ENiagaraSkelMeshTransforms=::NewEnumerator0
    Skeletal Mesh=$User.SkeletalMesh
    Socket Sampling Mode=::NewEnumerator5
   ) {module=SkeletalMeshLocation001}
   &ScaleSpriteSize(Uniform Curve Sprite Scale=DataInterface() {propertyValues {Curve {keys [
        {arriveTangent=6.4608, interpMode=RCIM_Cubic, leaveTangent=6.4608, tangentMode=RCTM_User, time=0, value=0}
        {interpMode=RCIM_Cubic, tangentMode=RCTM_User, time=0.25, value=1}
        {tangentMode=RCTM_User, time=1, value=0}
       ]}}})
   &CurlNoiseForce(Noise Frequency=200, Noise Strength=15, Pan Noise=true, Pan Noise Field=<0 0 1>)
   &Drag(Drag=2, Use Linear Drag=true, Use Rotational Drag=false)
   &SolveForcesAndVelocity(Clamp Velocity=false, Limit Acceleration=false)
   &ScaleSpriteSizeBySpeed(Max Scale Factor=<1.1 5>, Velocity Threshold=200)
   &Color(Color=&MultiplyLinearColor(
     Linear Color=&FN_Select_Color_from_Time_Of_Day(Morning=<rgba 1 1 1 1>, Day=<rgba 1 1 1 1>, Evening=<rgba 1 1 1 1>, Night=<rgba 0.5 0.5 0.5 1>)
     Scale Factor=&MakeVector4_FromLinearColor(LinearColor=&LerpLinearColors(StartColor=<rgba 0 3.5 2 1>, EndColor=<rgba 1 1 1 1>, LerpFactor=$Particles.NormalizedAge))
    ))
  ]
  rendererClasses [&NiagaraSpriteRendererProperties]
  renderers [{rendererClass=&NiagaraSpriteRendererProperties}]
 }
 GoopRibbon {
  emitterUpdateScript [
   &EmitterState()
   /Game/Effects/Niagara/Modules/Ribbons/FN_SpawnRibbon_v3(
    Number of Ribbons=4
    Skeleton Mesh:NiagaraDataInterfaceSkeletalMesh=$User.SkeletalMesh
    Spawn Mode:ENiagaraSkeletalSpawning=::NewEnumerator1
    SpawnRate=30
   )
   /SetVariables(Emitter.R2=0.33, Emitter.R3=0.5, Emitter.R4=0.66) {isSetParameters=true}
   &FN_EmitterDivingAttributes_v1(Accumulate Dive=-0.02, Accumulate Rest=0, Bone Index=2, SkeletalMesh:NiagaraDataInterfaceSkeletalMesh=$User.SkeletalMesh)
   /SetVariables(Emitter.AlphaPan:NiagaraFloat=&Lerp_Float(A=-0.55, B=-0.85, Alpha=&MakeCustomFloatFromBool(Bool=$System.IsDiving, True Float=1, False Float=0))) {isSetParameters=true}
   /SetVariables(Emitter.DistortionPannerSpeed:Vector3f=&MakeVector(X=$Emitter.AlphaPan, Y=0.1, Z=0)) {isSetParameters=true}
  ]
  particleSpawnScript [
   &FN_SetSpawnGroup()
   /SetVariables(Particles.IsFront:NiagaraInt32=&ModuloInt(A=$Particles.SpawnGroup, B=2)) {isSetParameters=true}
   /Game/Effects/Niagara/Modules/Ribbons/FN_InitializeRibbon_v5(
    Custom Offset=true
    Enable Skeletal Position=true
    Lifetime:NiagaraFloat=&MaskFloatBySpawnGroup(Masked Float=0.5, Passthrough Float=0.1, Spawn Group Mask=$Particles.IsFront)
    Local Vector Offset:Vector3f=&MaskVectorBySpawnGroup(
     Masked Vector=&MaskVectorBySpawnGroup(Masked Vector=<0 0 0>, Passthrough Vector=<0 0 0>)
     Passthrough Vector=&MaskVectorBySpawnGroup(Masked Vector=<1.75 -1.5 0>, Passthrough Vector=<-1.75 -1.5 0>, Spawn Group Mask=2)
     Spawn Group Mask=$Particles.IsFront
    )
    Offset Amount=-5
    Ribbon Width:NiagaraFloat=&MaskFloatBySpawnGroup(Masked Float=30, Passthrough Float=30, Spawn Group Mask=&ModuloInt(A=$Particles.SpawnGroup, B=2))
    SkeletalMesh:NiagaraDataInterfaceSkeletalMesh=$User.SkeletalMesh
    Write Material Random=false
   )
   &AddVelocity(Velocity=&FN_FrontEndVector3Modifier(Gameplay=<0 0 300>, Front End=<0 0 300>)) {module=AddVelocity001}
   /SetVariables(Particles.Color=/Niagara/DynamicInputs/LinearColor/MakeLinearColorFromFloat4(Vector:Vector4f=&MakeFloat4FromVectorAndFloat(
      "Vector (RGB)"=<1 1 1>
      "Float (Alpha)"=&MaskFloatBySpawnGroup(Masked Float=0, Passthrough Float=1, Spawn Group Mask=$Particles.IsFront)
     ))) {isSetParameters=true}
   /SetVariables(Particles.MaterialRandom:NiagaraFloat=&RandomRangeFloat(Minimum=0, Maximum=0)) {isSetParameters=true}
   /SetVariables() {isSetParameters=true}
   /SetVariables(Particles.Color2=&MakeLinearColorFromFloat4(Vector:Vector4f=&MakeFloat4FromVectorAndFloat(
      "Vector (RGB)"=<1 1 1>
      "Float (Alpha)"=&MaskFloatBySpawnGroup(Masked Float=1, Passthrough Float=0, Spawn Group Mask=$Particles.IsFront)
     ))) {isSetParameters=true}
  ]
  particleUpdateScript [
   &ScaleVelocity(Velocity Scale=&FN_FrontEndVector3Modifier(
     Gameplay=&MakeVector(
      X=0
      Y=0
      Z=&Multiply_Float(
       A=&Lerp_Float(A=0.9, B=0.7, Alpha=$System.Timeline.NormalizedTime)
       B=&FloatFromCurve(
        FloatCurve=DataInterface() {propertyValues {Curve {keys [<at 0 0.9>,<at 1 3>]}}}
        CurveIndex=&Divide_Float(A=&VectorLength(Vector=$Engine.Owner.Velocity), B=9000)
       )
      )
     )
     Front End=&MakeVector(X=0, Y=0, Z=1)
    ))
   &SolveForcesAndVelocity(Clamp Velocity=false, Limit Acceleration=false)
   /Game/Effects/Niagara/Modules/Ribbons/FN_UpdateRibbon_v6(Enable Skeletal Position=true, SkeletalMesh:NiagaraDataInterfaceSkeletalMesh=$User.SkeletalMesh)
   &DynamicMaterialParameters(
    Index 0 Param 1=&Add_Float(
     A=&MaskFloatBySpawnGroup(
      Masked Float=&MaskFloatBySpawnGroup(Masked Float=$Emitter.R1, Passthrough Float=$Emitter.R2)
      Passthrough Float=&MaskFloatBySpawnGroup(Masked Float=$Emitter.R3, Passthrough Float=$Emitter.R4, Spawn Group Mask=2)
      Spawn Group Mask=$Particles.IsFront
     )
     B=0
    )
    Index 0 Param 2=&FloatFromCurve(FloatCurve=DataInterface() {propertyValues {Curve {keys [
         {arriveTangent=0.0098, interpMode=RCIM_Cubic, leaveTangent=0.0098, tangentMode=RCTM_User, time=0.2, value=0.5}
         {arriveTangent=-0.2799, interpMode=RCIM_Cubic, leaveTangent=0.0167, tangentMode=RCTM_Break, time=1, value=0}
        ]}}})
    Index 0 Param 3=&Lerp_Float(A=-0.55, B=-1, Alpha=&MakeCustomFloatFromBool(Bool=$System.IsDiving, True Float=1, False Float=0))
    Module.Param0WriteEnabled=true
    Module.Param1WriteEnabled=true
    Module.Param2WriteEnabled=true
    Module.Param3WriteEnabled=true
   )
   &ScaleColor(
    Scale Alpha=&Multiply_Float(
     A=&FNTimeOfDayScalar(Night Scale=1.5, Morning Scale=2, Evening Scale=2)
     B=&FloatFromCurve(FloatCurve=DataInterface() {propertyValues {Curve {keys [<at 0 0>,<at 0.1 1>,<at 1 1>]}}})
    )
    Scale RGB=&Multiply_Vector(
     A=&MakeVectorFromLinearColorRGB(Color=<rgba 1 1 1 1>)
     B=&VectorFromFloat(Value=1.5)
    )
    ScaleA=true
    ScaleRGB=true
   )
   &ScaleRibbonWidth(Ribbon Width Scale=&Multiply_Float(
     A=&Multiply_Float(
      A=&Add_Float(
       A=&FloatFromCurve(FloatCurve=DataInterface() {propertyValues {Curve {keys [
            <at 0 0.3 RCIM_Cubic>
            {arriveTangent=1.7947, interpMode=RCIM_Cubic, leaveTangent=1.7947, tangentMode=RCTM_User, time=0.2, value=0.9}
            {arriveTangent=-1, interpMode=RCIM_Cubic, leaveTangent=-1, time=0.4, value=1.05}
            {arriveTangent=-0.6667, interpMode=RCIM_Cubic, leaveTangent=-0.6667, time=0.5, value=0.6}
            {arriveTangent=-1.5, interpMode=RCIM_Cubic, leaveTangent=-1.5, time=0.7, value=0.85}
            <at 0.9 0 RCIM_Cubic>
           ]}}})
       B=&Waveform(
        Global Phase=$Emitter.NormalizedLoopAge
        Global Amplitude Scale=0.5
        "Waveform [1]"=::NewEnumerator12
        "[1] Phase Scale"=1.25
        "[1] Phase"=&Add_Float(
         A=&Add_Float(
          A=$Emitter.Age
          B=&MaskFloatBySpawnGroup(
           Masked Float=&MaskFloatBySpawnGroup(Masked Float=$Emitter.R1, Passthrough Float=$Emitter.R2)
           Passthrough Float=&MaskFloatBySpawnGroup(Masked Float=$Emitter.R3, Passthrough Float=$Emitter.R4, Spawn Group Mask=2)
           Spawn Group Mask=$Particles.IsFront
          )
         )
         B=&OneMinusFloat(Float=$Particles.NormalizedAge)
        )
        "[1] Frequency"=5
        "[1] Amplitude Min/Max"=<0 1>
       )
      )
      B=&FloatFromCurve(FloatCurve=DataInterface() {propertyValues {Curve {keys [{interpMode=RCIM_Cubic, tangentMode=RCTM_User, time=0.75, value=1}, {interpMode=RCIM_Cubic, tangentMode=RCTM_User, time=1, value=0.5}]}}})
     )
     B=&MaskFloatBySpawnGroup(
      Masked Float=1
      Passthrough Float=&Waveform(
       Global Phase=&Add_Float(A=$Particles.NormalizedAge, B=&MakeFloatFromInt(Int=$Particles.SpawnGroup))
       "Waveform [1]"=::NewEnumerator12
       "[1] Phase Scale"=3.5
       "[1] Frequency"=1
       "[1] Amplitude Min/Max"=<0 2>
       "[1] Scale"=0.5
      )
      Spawn Group Mask=$Particles.IsFront
     )
    )) {module=ScaleRibbonWidth001}
  ]
  rendererClasses [&NiagaraRibbonRendererProperties]
  renderers [{rendererClass=&NiagaraRibbonRendererProperties}]
 }
 HandGoop {
  emitterUpdateScript [
   &EmitterState(
    Inactive Response=::NewEnumerator1
    Life Cycle Mode=::NewEnumerator1
    Loop Behavior=::NewEnumerator1
    Loop Duration Mode=::NewEnumerator1
    UseLoopDelay=false
   )
   &SpawnBurst_Instantaneous(Spawn Count=2, Use Spawn Probability=false, UseLoopCountLimit=false) {module=SpawnBurst_Instantaneous002}
  ]
  particleSpawnScript [
   &FN_SetSpawnGroup(Number of Spawn Groups=2)
   &InitializeParticle(Mesh Scale Mode=::NewEnumerator3, Mesh Uniform Scale=0.85, UsePositionOffset=false, Write Material Random=true)
  ]
  particleUpdateScript [
   &SkeletalMeshLocation(
    Bone / Socket Index=$Particles.SpawnGroup
    Bone Sampling Mode=::NewEnumerator4
    Inherit Velocity=false
    MaskSpawnGroup=false
    Skeletal Mesh=DataInterface() {propertyValues {
      DefaultMesh=null
      ExcludeBoneName=root
      OptionalComponentClass=null
      PreviewMesh=&M_Med_Soldier_04_Preview
      SkinningMode=SkinOnTheFly
      SoftSourceActor=null
      SourceMode=Default
      UvSetIndex=0
      WholeMeshLOD=-1
      bExcludeBone=false
      bReadDeformedGeometry=true
      bRequireCurrentFrameData=true
      ComponentTags []
      FilteredBones [hand_l,hand_r]
      FilteredSockets []
      MeshUserParameter {parameter {name=None, typeDefHandle {registeredTypeIndex=1}, varData []}}
      SamplingRegions []
     }}
   ) {module=SkeletalMeshLocation001}
   &ParticleState(Kill Particles When Lifetime Has Elapsed=false, Let Infinitely Lived Particles Die When Emitter Deactivates=true)
   &UpdateMeshOrientation(
    Facing Direction=&FN_FrontEndVector3Modifier(
     Gameplay=<0 0 0>
     Front End=&RotateVectorByQuaternion(Vector To Rotate=<0 0 1>, Quaternion=$Output.SkeletalMeshLocation001.SampledRotation)
    )
    Facing Interpolation=false
    Orientation Method=::NewEnumerator1
    Override Quat=false
    TransformFacing=false
    TransformUp=false
   )
   &SolveForcesAndVelocity(Clamp Velocity=false, Limit Acceleration=false)
   &DynamicMaterialParameters(
    Index 0 Param 1=&RandomRangeFloat(Minimum=0, Maximum=1)
    Index 0 Param 2=2
    Index 0 Param 3=1
    Index 0 Param 4=1
    Module.Param0WriteEnabled=true
    Module.Param1WriteEnabled=true
    Module.Param2WriteEnabled=true
    Module.Param3WriteEnabled=true
   )
  ]
  rendererClasses [&NiagaraMeshRendererProperties]
  renderers [{rendererClass=&NiagaraMeshRendererProperties, rendererIndex=0}]
 }
 MeshCenter {
  emitterUpdateScript [
   &EmitterState(Life Cycle Mode=::NewEnumerator1, Loop Behavior=::NewEnumerator1, Loop Duration Mode=::NewEnumerator1, UseLoopDelay=false)
   &SpawnBurst_Instantaneous(Spawn Count=2, Use Spawn Probability=false, UseLoopCountLimit=false)
  ]
  particleSpawnScript [
   &InitializeParticle(
    Lifetime Max=1.75
    Lifetime Min=0.8
    Lifetime Mode=::NewEnumerator1
    Mass Max=2
    Mass Mode=::NewEnumerator3
    Mesh Scale Mode=::NewEnumerator3
    Mesh Uniform Scale=&Multiply_Float(
     A=&SelectFloatFromArray(
      Direct Array Index=&ModuloInt(A=&ReturnExecIndex(), B=3)
      Float Selection Array=DataInterface() {propertyValues {FloatData [2.1,2.3,2.25]}}
     )
     B=0.8
    )
    UsePositionOffset=false
    Write Material Random=true
   )
   &InitialMeshOrientation(
    Mesh Orientation Mode=::NewEnumerator7
    Rotation=&RandomRangeVector(Minimum=<0 0 0>, Maximum=<1 0 0>)
    Use Rotation Vector=true
   )
   &FN_SetSpawnGroup(Number of Spawn Groups=2)
  ]
  particleUpdateScript [
   &ParticleState(
    Kill Particles When Lifetime Has Elapsed=false
    Let Infinitely Lived Particles Die When Emitter Deactivates=true
    Loop Particles Lifetime=true
   )
   &ScaleColor(
    Scale Alpha=&FloatFromCurve(FloatCurve=DataInterface() {propertyValues {Curve {keys [<at -0.0046 1.0113 RCIM_Cubic>,<at 1.0011 0.002 RCIM_Cubic>]}}})
    ScaleA=true
    ScaleRGB=true
   )
   &SkeletalMeshLocation(
    Bone / Socket Index=3
    Bone Sampling Mode=::NewEnumerator4
    MaskSpawnGroup=false
    Orientation Sampling=::NewEnumerator1
    Sampled Position Offset=&MakeVector(
     X=&SelectFloatFromArray(
      Direct Array Index=&ModuloInt(A=&ReturnExecIndex(), B=3)
      Float Selection Array=DataInterface() {propertyValues {FloatData [0,2.5,5]}}
     )
     Y=0
     Z=0
    )
    Skeletal Mesh=$User.SkeletalMesh
   )
   &UpdateMeshOrientation(Rotation Rate=&Multiply_Float(
     A=&RandomRangeFloat(Minimum=0.05, Maximum=0.1)
     B=&Waveform(
      "Waveform [1]"=::NewEnumerator12
      "[1] Phase"=&Add_Float(A=$Particles.Age, B=&RandomRangeFloat(Minimum=0, Maximum=1))
      "[1] Frequency"=1
      "[1] Amplitude Min/Max"=<0.5 1>
     )
    ))
   &UpdateMeshOrientation(
    Facing Direction=&RotateVectorByQuaternion(Vector To Rotate=<1 0 0>, Quaternion=$Particles.SkeletalMeshLocation.SampledRotation)
    Facing Interpolation=false
    Orientation Method=::NewEnumerator1
    Override Quat=false
    TransformFacing=false
    TransformUp=false
   ) {module=UpdateMeshOrientation001}
   &DynamicMaterialParameters(
    Index 0 Param 1=&Multiply_Float(
     A=&RandomRangeFloat(Minimum=0.5, Maximum=1)
     B=&SelectFloatFromArray(
      Direct Array Index=&ReturnExecIndex()
      Float Selection Array=DataInterface() {propertyValues {FloatData [0.5,4]}}
     )
    )
    Index 0 Param 2=&Multiply_Float(
     A=0.5
     B=&Waveform("Waveform [1]"=::NewEnumerator12, "[1] Frequency"=1, "[1] Amplitude Min/Max"=<0 1>, "[1] Scale"=0.5, "[1] Bias"=0.5)
    )
    Module.Param0WriteEnabled=true
    Module.Param1WriteEnabled=true
    Module.Param2WriteEnabled=true
    Module.Param3WriteEnabled=true
   )
  ]
  rendererClasses [&NiagaraMeshRendererProperties]
  renderers [{rendererClass=&NiagaraMeshRendererProperties, rendererIndex=0}]
 }
 MeshCenter001 {
  emitterUpdateScript [
   &EmitterState(Life Cycle Mode=::NewEnumerator1, Loop Behavior=::NewEnumerator1, Loop Duration Mode=::NewEnumerator1, UseLoopDelay=false)
   &SpawnBurst_Instantaneous(Use Spawn Probability=false, UseLoopCountLimit=false)
  ]
  particleSpawnScript [
   &InitializeParticle(
    Lifetime Max=1.75
    Lifetime Min=0.8
    Lifetime Mode=::NewEnumerator1
    Mass Max=2
    Mass Mode=::NewEnumerator3
    Mesh Scale Mode=::NewEnumerator3
    Mesh Uniform Scale=&Multiply_Float(
     A=&SelectFloatFromArray(
      Direct Array Index=&ModuloInt(A=&ReturnExecIndex(), B=3)
      Float Selection Array=DataInterface() {propertyValues {FloatData [2.7,2.3,2.25]}}
     )
     B=0.8
    )
    UsePositionOffset=false
    Write Material Random=true
   )
   &InitialMeshOrientation(
    Mesh Orientation Mode=::NewEnumerator7
    Rotation=&RandomRangeVector(Minimum=<0 0 0>, Maximum=<1 0 0>)
    Use Rotation Vector=true
   )
  ]
  particleUpdateScript [
   &ParticleState(
    Kill Particles When Lifetime Has Elapsed=false
    Let Infinitely Lived Particles Die When Emitter Deactivates=true
    Loop Particles Lifetime=true
   )
   &ScaleMeshSize(Scale Factor=&VectorFromFloat(Value=&Multiply_Float(
      A=&Waveform("Waveform [1]"=::NewEnumerator12, "[1] Frequency"=1, "[1] Amplitude Min/Max"=<0 0.3>, "[1] Bias"=0.75)
      B=1
     )))
   &ScaleColor(
    Scale Alpha=&FloatFromCurve(FloatCurve=DataInterface() {propertyValues {Curve {keys [<at -0.0046 1.0113 RCIM_Cubic>,<at 1.0011 0.002 RCIM_Cubic>]}}})
    ScaleA=true
    ScaleRGB=true
   )
   &SkeletalMeshLocation(
    Bone / Socket Index=3
    Bone Sampling Mode=::NewEnumerator4
    MaskSpawnGroup=false
    Orientation Sampling=::NewEnumerator1
    Sampled Position Offset=&MakeVector(
     X=&SelectFloatFromArray(
      Direct Array Index=&ModuloInt(A=&ReturnExecIndex(), B=3)
      Float Selection Array=DataInterface() {propertyValues {FloatData [2.5,2.5,5]}}
     )
     Y=0
     Z=0
    )
    Skeletal Mesh=$User.SkeletalMesh
   )
   &UpdateMeshOrientation(Rotation Rate=&Multiply_Float(
     A=&RandomRangeFloat(Minimum=0.05, Maximum=0.1)
     B=&Waveform(
      "Waveform [1]"=::NewEnumerator12
      "[1] Phase"=&Add_Float(A=$Particles.Age, B=&RandomRangeFloat(Minimum=0, Maximum=1))
      "[1] Frequency"=1
      "[1] Amplitude Min/Max"=<0.5 1>
     )
    ))
   &UpdateMeshOrientation(
    Facing Direction=&RotateVectorByQuaternion(Vector To Rotate=<1 0 0>, Quaternion=$Particles.SkeletalMeshLocation.SampledRotation)
    Facing Interpolation=false
    Orientation Method=::NewEnumerator1
    Override Quat=false
    TransformFacing=false
    TransformUp=false
   ) {module=UpdateMeshOrientation001}
   &DynamicMaterialParameters(
    Index 0 Param 1=&Multiply_Float(A=1, B=1)
    Index 0 Param 2=&Multiply_Float(A=&Waveform("Waveform [1]"=::NewEnumerator12, "[1] Frequency"=1, "[1] Amplitude Min/Max"=<0 1>), B=10)
    Module.Param0WriteEnabled=true
    Module.Param1WriteEnabled=true
    Module.Param2WriteEnabled=true
    Module.Param3WriteEnabled=true
   )
  ]
  rendererClasses [&NiagaraMeshRendererProperties]
  renderers [{rendererClass=&NiagaraMeshRendererProperties, rendererIndex=0}]
 }
 YellowRibbon_01 {
  emitterUpdateScript [
   &EmitterState(Life Cycle Mode=::NewEnumerator1, Loop Behavior=::NewEnumerator1, Loop Duration Mode=::NewEnumerator1, UseLoopDelay=false)
   &FN_SpawnRibbon_v2(Number of Ribbons=2, SpawnRate=50)
   &FN_EmitterDivingAttributes_v1(Accumulate Dive=-0.0025, Accumulate Rest=0, SkeletalMesh:NiagaraDataInterfaceSkeletalMesh=$User.SkeletalMesh)
   /SetVariables(
    Emitter.R1:NiagaraFloat=&RandomRangeFloat(Minimum=0, Maximum=1)
    Emitter.R2:NiagaraFloat=&RandomRangeFloat(Minimum=0, Maximum=1)
    Emitter.R3:NiagaraFloat=&RandomRangeFloat(Minimum=0, Maximum=1)
    Emitter.R4:NiagaraFloat=&RandomRangeFloat(Minimum=0, Maximum=1)
   ) {isSetParameters=true}
   /SetVariables(Emitter.AlphaPan:NiagaraFloat=&Lerp_Float(A=-0.3, B=-0.35, Alpha=&MakeCustomFloatFromBool(Bool=$System.IsDiving, True Float=1, False Float=0))) {isSetParameters=true}
   /SetVariables(Emitter.DistortionPan:NiagaraFloat=&Lerp_Float(A=-0.03, B=-0.035, Alpha=&MakeCustomFloatFromBool(Bool=$System.IsDiving, True Float=1, False Float=0))) {isSetParameters=true}
  ]
  particleSpawnScript [
   &FN_SetSpawnGroup(Number of Spawn Groups=2)
   /SetVariables(Particles.IsFront:NiagaraInt32=&ModuloInt(A=$Particles.SpawnGroup, B=2)) {isSetParameters=true, module=SetVariables_#5}
   &FN_InitializeRibbon_v3(
    Color:LinearColor=&MultiplyLinearColor(Linear Color=<rgba 0.7991 1 0.1413 1>, Scale Factor=<1 1 1 1>)
    Custom Offset=true
    Enable Skeletal Position=true
    Lifetime:NiagaraFloat=&MaskFloatBySpawnGroup(Masked Float=0.5, Passthrough Float=0.25, Spawn Group Mask=$Particles.IsFront)
    Local Vector Offset:Vector3f=&MaskVectorBySpawnGroup(
     Masked Vector=&MaskVectorBySpawnGroup(Masked Vector=<-3.5 0 0>, Passthrough Vector=<3.5 0 0>)
     Passthrough Vector=&MaskVectorBySpawnGroup(Masked Vector=<-2 0 0>, Passthrough Vector=<2 0 0>, Spawn Group Mask=2)
     Spawn Group Mask=$Particles.IsFront
    )
    Offset Amount=1
    Ribbon Width:NiagaraFloat=&MaskFloatBySpawnGroup(Masked Float=20, Passthrough Float=10, Spawn Group Mask=$Particles.IsFront)
   )
   /SetVariables(Particles.RibbonNumber:NiagaraInt32=$Particles.SpawnGroup) {isSetParameters=true, module=SetVariables_#6}
   &AddVelocity(Velocity=&Lerp_Vector(A=<0 0 400>, B=<0 0 400>, Alpha=$Emitter.Dive))
   /SetVariables(Particles.SmokeColor:LinearColor=<rgba 0 0 0 0.6>) {isSetParameters=true, module=SetVariables_#7}
  ]
  particleUpdateScript [
   &ScaleVelocity(Velocity Scale=&MakeVector(
     X=&FloatFromCurve(FloatCurve=DataInterface() {propertyValues {Curve {keys [<at 0 1 RCIM_Cubic>,<at 0.3 0 RCIM_Cubic>]}}})
     Y=&FloatFromCurve(FloatCurve=DataInterface() {propertyValues {Curve {keys [<at 0 1 RCIM_Cubic>,<at 0.3 0 RCIM_Cubic>]}}})
     Z=1
    ))
   &DynamicMaterialParameters(
    Index 0 Param 1=&Add_Float(
     A=&MaskFloatBySpawnGroup(
      Masked Float=&MaskFloatBySpawnGroup(Masked Float=$Emitter.R1, Passthrough Float=$Emitter.R2)
      Passthrough Float=&MaskFloatBySpawnGroup(Masked Float=$Emitter.R3, Passthrough Float=$Emitter.R4, Spawn Group Mask=2)
      Spawn Group Mask=$Particles.IsFront
     )
     B=$Emitter.IntegratedDive
    )
    Index 0 Param 3=&Lerp_Float(A=1, B=0.75, Alpha=$Emitter.Dive)
    Index 2 Param 1=&MaskFloatBySpawnGroup(Masked Float=1, Passthrough Float=0.5, Spawn Group Mask=$Particles.IsFront)
    Index 2 Param 2=&Add_Float(
     A=&MaskFloatBySpawnGroup(
      Masked Float=&MaskFloatBySpawnGroup(Masked Float=$Emitter.R1, Passthrough Float=$Emitter.R2)
      Passthrough Float=&MaskFloatBySpawnGroup(Masked Float=$Emitter.R3, Passthrough Float=$Emitter.R4, Spawn Group Mask=2)
      Spawn Group Mask=$Particles.IsFront
     )
     B=0
    )
    Module.Param0WriteEnabled=true
    Module.Param0WriteEnabled2=true
    Module.Param1WriteEnabled=true
    Module.Param1WriteEnabled2=true
    Module.Param2WriteEnabled=true
    Module.Param2WriteEnabled2=true
    Module.Param3WriteEnabled=true
    Module.Param3WriteEnabled2=true
    Write Parameter Index 2=true
   ) {module=DynamicMaterialParameters001}
   &SolveForcesAndVelocity(Clamp Velocity=false, Limit Acceleration=false)
   &FN_UpdateRibbon(
    Enable Skeletal Position=true
    Initial Color:LinearColor=$Particles.Initial.Color
    Initial RibbonWidth:NiagaraFloat=$Particles.Initial.RibbonWidth
   )
   /SetVariables(Particles.Color=&ColorFromCurve(ColorCurve=DataInterface() {propertyValues {
       AlphaCurve {keys [<at 0 1 RCIM_Cubic>]}
       BlueCurve {keys [<at 0 0.5312 RCIM_Cubic>,<at 0.998 0.2083>]}
       GreenCurve {keys [<at 0 2 RCIM_Cubic>,<at 0.998 2>]}
       RedCurve {keys [<at 0 1.9869 RCIM_Cubic>,<at 0.998 1.9841>]}
      }})) {isSetParameters=true, module=SetVariables_#8}
   &ScaleColor(
    Color Value To Scale=$Particles.Color
    Scale RGB=&VectorFromFloat(Value=&FloatFromCurve(FloatCurve=DataInterface() {propertyValues {
        bOverrideOptimizeThreshold=true
        Curve {keys [
          {arriveTangent=0.1804, interpMode=RCIM_Cubic, leaveTangent=0.1804, tangentMode=RCTM_User, time=0, value=6}
          {arriveTangent=-0.4145, interpMode=RCIM_Cubic, leaveTangent=-0.4145, tangentMode=RCTM_User, time=0.3, value=1}
         ]}
       }}))
    ScaleA=true
    ScaleRGB=true
   ) {module=ScaleColor001}
   &ScaleRibbonWidth(Ribbon Width Scale=&FloatFromCurve(
     FloatCurve=DataInterface() {propertyValues {Curve {keys [
         <at 0 0.2 RCIM_Cubic>
         {arriveTangent=3.8, interpMode=RCIM_Cubic, leaveTangent=3.8, tangentMode=RCTM_User, time=0.2, value=0.8}
         {arriveTangent=-1, interpMode=RCIM_Cubic, leaveTangent=-1, time=0.5, value=1.025}
         <at 1 0 RCIM_Cubic>
        ]}}}
     Scale Curve=2
    ))
  ]
  rendererClasses [&NiagaraRibbonRendererProperties]
  renderers [{rendererClass=&NiagaraRibbonRendererProperties}]
 }
 YellowRibbon_02 {
  emitterUpdateScript [
   &EmitterState(Life Cycle Mode=::NewEnumerator1, Loop Behavior=::NewEnumerator1, Loop Duration Mode=::NewEnumerator1, UseLoopDelay=false)
   &FN_SpawnRibbon_v2(Number of Ribbons=2, SpawnRate=50)
   &FN_EmitterDivingAttributes_v1(Accumulate Dive=-0.0025, Accumulate Rest=0, SkeletalMesh:NiagaraDataInterfaceSkeletalMesh=$User.SkeletalMesh)
   /SetVariables(Emitter.R2=0.33, Emitter.R3=0.5, Emitter.R4=0.66) {isSetParameters=true}
   /SetVariables(Emitter.AlphaPan:NiagaraFloat=&Lerp_Float(A=-0.3, B=-0.35, Alpha=&MakeCustomFloatFromBool(Bool=$System.IsDiving, True Float=1, False Float=0))) {isSetParameters=true}
   /SetVariables(Emitter.DistortionPan:NiagaraFloat=&Lerp_Float(A=-0.03, B=-0.035, Alpha=&MakeCustomFloatFromBool(Bool=$System.IsDiving, True Float=1, False Float=0))) {isSetParameters=true}
  ]
  particleSpawnScript [
   &FN_SetSpawnGroup(Number of Spawn Groups=2)
   /SetVariables(Particles.IsFront:NiagaraInt32=&ModuloInt(A=$Particles.SpawnGroup, B=2)) {isSetParameters=true, module=SetVariables_#5}
   &FN_InitializeRibbon_v3(
    Color:LinearColor=&MultiplyLinearColor(Linear Color=<rgba 0.7991 1 0.1413 1>, Scale Factor=<1 1 1 1>)
    Custom Offset=true
    Enable Skeletal Position=true
    Lifetime:NiagaraFloat=&MaskFloatBySpawnGroup(Masked Float=0.35, Passthrough Float=0.175, Spawn Group Mask=$Particles.IsFront)
    Local Vector Offset:Vector3f=&MaskVectorBySpawnGroup(
     Masked Vector=&MaskVectorBySpawnGroup(Masked Vector=<-2.5 0 0>, Passthrough Vector=<2.5 0 0>)
     Passthrough Vector=&MaskVectorBySpawnGroup(Masked Vector=<-1 0 0>, Passthrough Vector=<1 0 0>, Spawn Group Mask=2)
     Spawn Group Mask=$Particles.IsFront
    )
    Offset Amount=1
    Ribbon Width:NiagaraFloat=&MaskFloatBySpawnGroup(Masked Float=7.5, Passthrough Float=3.5, Spawn Group Mask=$Particles.IsFront)
   )
   /SetVariables(Particles.RibbonNumber:NiagaraInt32=$Particles.SpawnGroup) {isSetParameters=true, module=SetVariables_#6}
   &AddVelocity(Velocity=&Lerp_Vector(A=<0 0 400>, B=<0 0 400>, Alpha=$Emitter.Dive))
   /SetVariables(Particles.SmokeColor:LinearColor=<rgba 0 0 0 0.6>) {isSetParameters=true, module=SetVariables_#7}
  ]
  particleUpdateScript [
   &ScaleVelocity(Velocity Scale=&MakeVector(
     X=&FloatFromCurve(FloatCurve=DataInterface() {propertyValues {Curve {keys [<at 0 1 RCIM_Cubic>,<at 0.3 0 RCIM_Cubic>]}}})
     Y=&FloatFromCurve(FloatCurve=DataInterface() {propertyValues {Curve {keys [<at 0 1 RCIM_Cubic>,<at 0.3 0 RCIM_Cubic>]}}})
     Z=1
    ))
   &DynamicMaterialParameters(
    Index 0 Param 1=&Add_Float(
     A=&MaskFloatBySpawnGroup(
      Masked Float=&MaskFloatBySpawnGroup(Masked Float=$Emitter.R1, Passthrough Float=$Emitter.R2)
      Passthrough Float=&MaskFloatBySpawnGroup(Masked Float=$Emitter.R3, Passthrough Float=$Emitter.R4, Spawn Group Mask=2)
      Spawn Group Mask=$Particles.IsFront
     )
     B=$Emitter.IntegratedDive
    )
    Index 0 Param 3=&Lerp_Float(A=1, B=0.2, Alpha=$Emitter.Dive)
    Index 2 Param 1=&MaskFloatBySpawnGroup(Masked Float=1, Passthrough Float=0.5, Spawn Group Mask=$Particles.IsFront)
    Index 2 Param 2=&Add_Float(
     A=&MaskFloatBySpawnGroup(
      Masked Float=&MaskFloatBySpawnGroup(Masked Float=$Emitter.R1, Passthrough Float=$Emitter.R2)
      Passthrough Float=&MaskFloatBySpawnGroup(Masked Float=$Emitter.R3, Passthrough Float=$Emitter.R4, Spawn Group Mask=2)
      Spawn Group Mask=$Particles.IsFront
     )
     B=0
    )
    Module.Param0WriteEnabled=true
    Module.Param0WriteEnabled2=true
    Module.Param1WriteEnabled=true
    Module.Param1WriteEnabled2=true
    Module.Param2WriteEnabled=true
    Module.Param2WriteEnabled2=true
    Module.Param3WriteEnabled=true
    Module.Param3WriteEnabled2=true
    Write Parameter Index 2=true
   ) {module=DynamicMaterialParameters001}
   &CurlNoiseForce(Noise Frequency=20, Noise Strength=100, Pan Noise=false)
   &SolveForcesAndVelocity(Clamp Velocity=false, Limit Acceleration=false)
   &FN_UpdateRibbon(
    Enable Skeletal Position=true
    Initial Color:LinearColor=$Particles.Initial.Color
    Initial RibbonWidth:NiagaraFloat=$Particles.Initial.RibbonWidth
   )
   /SetVariables(Particles.Color=&ColorFromCurve(ColorCurve=DataInterface() {propertyValues {
       AlphaCurve {keys [<at 0 1 RCIM_Cubic>]}
       BlueCurve {keys [<at 0 0.5312 RCIM_Cubic>,<at 0.998 0.2083>]}
       GreenCurve {keys [<at 0 2 RCIM_Cubic>,<at 0.998 2>]}
       RedCurve {keys [<at 0 1.9869 RCIM_Cubic>,<at 0.998 1.9841>]}
      }})) {isSetParameters=true, module=SetVariables_#8}
   &ScaleColor(
    Color Value To Scale=$Particles.Color
    Scale RGB=&VectorFromFloat(Value=&FloatFromCurve(FloatCurve=DataInterface() {propertyValues {
        bOverrideOptimizeThreshold=true
        Curve {keys [
          {arriveTangent=0.1804, interpMode=RCIM_Cubic, leaveTangent=0.1804, tangentMode=RCTM_User, time=0, value=6}
          {arriveTangent=-0.4145, interpMode=RCIM_Cubic, leaveTangent=-0.4145, tangentMode=RCTM_User, time=0.3, value=1}
         ]}
       }}))
    ScaleA=true
    ScaleRGB=true
   ) {module=ScaleColor001}
   &ScaleRibbonWidth(Ribbon Width Scale=&FloatFromCurve(
     FloatCurve=DataInterface() {propertyValues {Curve {keys [
         <at 0 0.25 RCIM_Cubic>
         {arriveTangent=3.8, interpMode=RCIM_Cubic, leaveTangent=3.8, tangentMode=RCTM_User, time=0.2, value=0.8}
         {arriveTangent=-1, interpMode=RCIM_Cubic, leaveTangent=-1, time=0.5, value=1.025}
         <at 1 0 RCIM_Cubic>
        ]}}}
     Scale Curve=6
    ))
  ]
  rendererClasses [&NiagaraRibbonRendererProperties]
  renderers [{rendererClass=&NiagaraRibbonRendererProperties}]
 }
}
summary {
 systemName=NS_Contrail_NitroFlow
 emitters [
  {emitterName=DripsHands, rendererClasses [&NiagaraSpriteRendererProperties]}
  {emitterName=MeshCenter, rendererClasses [&NiagaraMeshRendererProperties]}
  {emitterName=DripsCenter, rendererClasses [&NiagaraSpriteRendererProperties]}
  {emitterName=HandGoop, rendererClasses [&NiagaraMeshRendererProperties]}
  {emitterName=YellowRibbon_02, rendererClasses [&NiagaraRibbonRendererProperties]}
  {emitterName=YellowRibbon_01, rendererClasses [&NiagaraRibbonRendererProperties]}
  {emitterName=MeshCenter001, rendererClasses [&NiagaraMeshRendererProperties]}
  {emitterName=GoopRibbon, rendererClasses [&NiagaraRibbonRendererProperties]}
  {emitterName=CentralRibbons, rendererClasses [&NiagaraRibbonRendererProperties]}
 ]
 userVariables [
  User.bIsFrontEnd {type=NiagaraBool}
  User.bIsFrontEndPreview {type=NiagaraBool, defaultValue=true}
  User.GameTimeFromBP {type=NiagaraFloat, defaultValue=0, description=""}
  User.ParaGlideLeanAlpha {type=NiagaraFloat, defaultValue=0, description=""}
  User.RibbonIDOffset {type=NiagaraInt32}
  User.SkeletalMesh {
   type=NiagaraDataInterfaceSkeletalMesh
   defaultValue {
    struct=NiagaraExt_VariableValue_DataInterface
    value {dataInterface=&self:NiagaraDataInterfaceSkeletalMesh_0}
   }
  }
  User.Skelmesh {
   type=NiagaraDataInterfaceSkeletalMesh
   defaultValue {
    struct=NiagaraExt_VariableValue_DataInterface
    value {dataInterface=&self:NiagaraDataInterfaceSkeletalMesh_1}
   }
  }
 ]
}
